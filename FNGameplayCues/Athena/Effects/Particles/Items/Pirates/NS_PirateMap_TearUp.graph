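# Niagara gameplay-cue effect of a pirate map tearing apart, built as timed one-shot bursts around a shared system offset.
# The Map mesh stretches horizontally with a jitter for 0.4 seconds, driven by scale and material curves.
# As it ends, two mirrored mesh halves fly apart, shrinking and fading.
# Staggered delays add a mesh flash, falling ash sprites, GPU curl-noise confetti and tinted smoke and swirl sprites at the tear.
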NiagaraSystem /FNGameplayCues/Athena/Effects/Particles/Items/Pirates/NS_PirateMap_TearUp
dependencies {
 usedDataInterfaces [
  &NiagaraDataInterfaceVectorCurve
  &NiagaraDataInterfaceCurve
  &NiagaraDataInterfaceSpriteRendererInfo
 ]
 usedDynamicInputs [&ConvertVectorToPosition, &AddVectorToPosition]
 usedModules [
  &self:SystemSpawnScript.SystemScriptSource.SystemScriptGraph.NiagaraNodeAssignment_0.SetVariables_#1
  &SystemState
 ]
}
emitters {
 Ash {
  emitterUpdateScript [
   &EmitterState(Life Cycle Mode=::NewEnumerator1, Loop Behavior=::NewEnumerator1, Loop Delay=0.4, Loop Duration=0.15, UseLoopDelay=true)
   &SpawnBurst_Instantaneous(
    Spawn Count=&RandomRangeInt(Minimum=5, Maximum=6)
    Use Spawn Probability=false
    UseLoopCountLimit=false
   )
  ]
  particleSpawnScript [
   /SetVariables(Particles.SpriteRotationRate=&RandomRangeFloat(Minimum=-1, Maximum=1)) {isSetParameters=true}
   &InitializeParticle(
    Lifetime=&RandomRangeFloat(Minimum=1.4, Maximum=2)
    Position Offset=&Add_Vector(A=&RandomRangeVector(Minimum=<-6 -1.5 0>, Maximum=<6 1.5 0>), B=$System.SystemOffset)
    Position Offset Coordinate Space=::Local
    Sprite Rotation Angle=&RandomRangeFloat(Minimum=0, Maximum=1)
    Sprite Rotation Mode=::NewEnumerator1
    Sprite Size=&Vector2DFromFloat(Value=&RandomRangeFloat(Minimum=6, Maximum=12))
    Sprite Size Mode=::NewEnumerator4
    UsePositionOffset=true
    Write Material Random=true
   )
   &Color(Color=&MakeLinearColorFromVectorAndFloat("Vector (RGB)"=<1 1 1>, "Float (Alpha)"=&ClampFloat(Float=1)))
   &AddVelocity(Velocity=&RandomRangeVector(Minimum=<-100 -30 5>, Maximum=<100 30 5>))
  ]
  particleUpdateScript [
   &ParticleState()
   &SpriteRotationRate(Rotation Rate=&AngleConversion(
     Angle:NiagaraFloat=$Particles.SpriteRotationRate
     Angle Input:ENiagara_AngleInput=::NewEnumerator1
     Angle Output:ENiagara_AngleInput=::NewEnumerator0
    ))
   &ScaleSpriteSize(
    Scale Sprite Size Mode=::NewEnumerator0
    Uniform Scale Factor=&FloatFromCurve(FloatCurve=DataInterface() {propertyValues {Curve {keys [
         <at 0 0 RCIM_Cubic>
         {arriveTangent=16.373, interpMode=RCIM_Cubic, leaveTangent=16.373, time=0.02, value=0.6769}
         {arriveTangent=-0.6907, interpMode=RCIM_Cubic, leaveTangent=-0.6907, time=0.0613, value=1.0035}
         <at 1 0>
        ]}}})
   )
   &ScaleColor(ScaleA=true, ScaleRGB=true)
   /Niagara/Modules/Update/SubUV/SubUVAnimation(Number Of Frames=40)
   &AccelerationForce(Acceleration=<0 0 -80>)
   &SolveForcesAndVelocity(Clamp Velocity=false, Limit Acceleration=false)
   &DynamicMaterialParameters(
    Index 0 Param 2=&FloatFromCurve(FloatCurve=DataInterface() {propertyValues {Curve {keys [<at 0 0>,<at 1 1>]}}})
    Index 0 Param 3=1
    Index 0 Param 4=1
    Module.Index0Param0Write=true
    Module.Index0Param1Write=true
    Module.Index0Param2Write=true
    Module.Index0Param3Write=true
   )
  ]
  rendererClasses [&NiagaraSpriteRendererProperties]
  renderers [{rendererClass=&NiagaraSpriteRendererProperties}]
 }
 Confetti_Center {
  simTarget=GPUComputeSim
  emitterUpdateScript [
   &EmitterState(Life Cycle Mode=::NewEnumerator1, Loop Behavior=::NewEnumerator1, Loop Delay=0.385, Loop Duration=0.2, UseLoopDelay=true)
   &SpawnBurst_Instantaneous(
    Spawn Count=&RandomRangeInt(Minimum=25, Maximum=50)
    Use Spawn Probability=false
    UseLoopCountLimit=false
   )
  ]
  particleSpawnScript [
   /SetVariables(Particles.SpriteRotationRate=&RandomRangeFloat(Minimum=-2, Maximum=2)) {isSetParameters=true}
   &InitializeParticle(
    Lifetime=&RandomRangeFloat(Minimum=1, Maximum=2)
    Position Offset=&Add_Vector(A=&RandomRangeVector(Minimum=<-6 -1.5 0>, Maximum=<6 1.5 0>), B=$System.SystemOffset)
    Position Offset Coordinate Space=::Local
    Sprite Rotation Angle=&RandomRangeFloat(Minimum=0, Maximum=1)
    Sprite Rotation Mode=::NewEnumerator1
    Sprite Size=&Vector2DFromFloat(Value=&RandomRangeFloat(Minimum=1, Maximum=2))
    Sprite Size Mode=::NewEnumerator4
    UsePositionOffset=true
    Write Material Random=true
   )
   &Color(Color=&MakeLinearColorFromVectorAndFloat("Vector (RGB)"=<1.25 1.25 1.25>, "Float (Alpha)"=&ClampFloat(Float=1)))
  ]
  particleUpdateScript [
   &ParticleState()
   &SpriteRotationRate(Rotation Rate=&AngleConversion(
     Angle:NiagaraFloat=$Particles.SpriteRotationRate
     Angle Input:ENiagara_AngleInput=::NewEnumerator1
     Angle Output:ENiagara_AngleInput=::NewEnumerator0
    ))
   &ScaleSpriteSize(
    Scale Sprite Size Mode=::NewEnumerator0
    Uniform Scale Factor=&FloatFromCurve(FloatCurve=DataInterface() {propertyValues {Curve {keys [<at 0.8 1 RCIM_Cubic>,<at 1 0>]}}})
   )
   &Drag(Drag=12, Use Linear Drag=true, Use Rotational Drag=true)
   &AccelerationForce(Acceleration=&RandomRangeVector(Minimum=<0 0 -550>, Maximum=<0 0 -1200>))
   &CurlNoiseForce(Noise Strength=550, Pan Noise=false)
   &SolveForcesAndVelocity(Clamp Velocity=false, Limit Acceleration=false)
  ]
  rendererClasses [&NiagaraSpriteRendererProperties]
  renderers [{rendererClass=&NiagaraSpriteRendererProperties}]
 }
 Flash {
  emitterUpdateScript [
   &EmitterState(Life Cycle Mode=::NewEnumerator1, Loop Behavior=::NewEnumerator1, Loop Delay=0.38, Loop Duration=1, UseLoopDelay=true)
   &SpawnBurst_Instantaneous(Use Spawn Probability=false, UseLoopCountLimit=false)
  ]
  particleSpawnScript [
   &InitializeParticle(
    Lifetime=0.125
    Mesh Scale=&VectorFromFloat(Value=&RandomRangeFloat(Minimum=0.5, Maximum=1.3))
    Mesh Scale Mode=::NewEnumerator4
    Position Offset=&Add_Vector(A=&RandomRangeVector(Minimum=<-1 -1.5 0>, Maximum=<1 1.5 0>), B=$System.SystemOffset)
    UsePositionOffset=true
    Write Material Random=true
   )
   &Color(Color=&MakeLinearColorFromVectorAndFloat("Vector (RGB)"=<1 1 1>, "Float (Alpha)"=&ClampFloat(Float=12)))
   /Niagara/Modules/CascadeConversion/CascadeConversion_InitialMeshRotation(Rotation:Vector3f=&RandomRangeVector(Minimum=<0 0 0>, Maximum=<1 1 1>))
  ]
  particleUpdateScript [
   &ParticleState()
   &ScaleMeshSize(Scale Factor=&VectorFromFloat(Value=&FloatFromCurve(FloatCurve=DataInterface() {propertyValues {Curve {keys [{arriveTangent=1.943, interpMode=RCIM_Cubic, time=0, value=0.2203}, <at 1 1>]}}})))
   &DynamicMaterialParameters(
    Index 0 Param 1=&FloatFromCurve(FloatCurve=DataInterface() {propertyValues {Curve {keys [<at 0 0 RCIM_Cubic>, {arriveTangent=1.2213, leaveTangent=1.2213, time=1, value=0.6}]}}})
    Index 0 Param 2=250
    Index 0 Param 3=1
    Index 0 Param 4=1
    Module.Index0Param0Write=true
    Module.Index0Param1Write=true
    Module.Index0Param2Write=true
    Module.Index0Param3Write=true
   )
   &ScaleColor(
    Scale Alpha=&FloatFromCurve(FloatCurve=DataInterface() {propertyValues {Curve {keys [<at 0 0.125>,<at 1 0>]}}})
    ScaleA=true
    ScaleRGB=true
   )
   &CameraOffset(Camera Offset Amount=1)
  ]
  rendererClasses [&NiagaraMeshRendererProperties]
  renderers [{rendererClass=&NiagaraMeshRendererProperties, rendererIndex=0}]
 }
 Map {
  emitterUpdateScript [
   &EmitterState(Life Cycle Mode=::NewEnumerator1, Loop Behavior=::NewEnumerator1, Loop Duration=1, UseLoopDelay=true)
   &SpawnBurst_Instantaneous(Use Spawn Probability=false, UseLoopCountLimit=false)
  ]
  particleSpawnScript [&InitializeParticle(
    Lifetime=0.4
    Mesh Scale=<0.42 0.44 0.44>
    Mesh Scale Mode=::NewEnumerator4
    Position Offset=$System.SystemOffset
    UsePositionOffset=true
    Write Material Random=true
   )]
  particleUpdateScript [
   &ParticleState()
   &ScaleMeshSize(Scale Factor=&VectorFromCurve(VectorCurve=DataInterface() {propertyValues {
       XCurve {keys [
         {arriveTangent=-0.9224, interpMode=RCIM_Cubic, time=0, value=1}
         {arriveTangent=-0.0048, interpMode=RCIM_Cubic, leaveTangent=-0.0048, time=0.0444, value=0.974}
         {arriveTangent=0.2841, interpMode=RCIM_Cubic, leaveTangent=0.2841, time=0.2251, value=0.9989}
         {arriveTangent=0.1717, interpMode=RCIM_Cubic, leaveTangent=0.1717, time=0.4866, value=1.0996}
         {arriveTangent=0.4993, interpMode=RCIM_Cubic, leaveTangent=0.4993, time=0.5411, value=1.0532}
         {arriveTangent=0.2252, interpMode=RCIM_Cubic, leaveTangent=0.2252, time=0.5701, value=1.1413}
         {arriveTangent=0.0841, interpMode=RCIM_Cubic, leaveTangent=0.0841, time=0.6101, value=1.0687}
         {arriveTangent=-0.0975, interpMode=RCIM_Cubic, leaveTangent=-0.0975, time=0.6417, value=1.1473}
         {arriveTangent=-0.1927, interpMode=RCIM_Cubic, leaveTangent=-0.1927, time=0.6936, value=1.0606}
         {arriveTangent=0.4634, interpMode=RCIM_Cubic, leaveTangent=0.4634, time=0.7303, value=1.1302}
         {arriveTangent=-0.0538, interpMode=RCIM_Cubic, leaveTangent=-0.0538, time=0.7788, value=1.1001}
         {arriveTangent=1.4495, interpMode=RCIM_Cubic, leaveTangent=1.4495, time=0.9058, value=1.1208}
         <at 1 1.4207>
        ]}
       YCurve {keys [<at 0 1 RCIM_Cubic>,<at 0.0444 1 RCIM_Cubic>,<at 0.2251 1 RCIM_Cubic>,<at 0.4866 1 RCIM_Cubic>,<at 0.5411 1 RCIM_Cubic>,<at 0.5701 1 RCIM_Cubic>,<at 0.6101 1 RCIM_Cubic>,<at 0.6417 1 RCIM_Cubic>,<at 0.6936 1 RCIM_Cubic>,<at 0.7303 1 RCIM_Cubic>,<at 0.7788 1 RCIM_Cubic>,<at 0.9058 1 RCIM_Cubic>,<at 1 1>]}
       ZCurve {keys [<at 0 1 RCIM_Cubic>,<at 0.0444 1 RCIM_Cubic>,<at 0.2251 1 RCIM_Cubic>,<at 0.4866 1 RCIM_Cubic>,<at 0.5411 1 RCIM_Cubic>,<at 0.5701 1 RCIM_Cubic>,<at 0.6101 1 RCIM_Cubic>,<at 0.6417 1 RCIM_Cubic>,<at 0.6936 1 RCIM_Cubic>,<at 0.7303 1 RCIM_Cubic>,<at 0.7788 1 RCIM_Cubic>,<at 0.9058 1 RCIM_Cubic>,<at 1 1>]}
      }}))
   &Color(Color=&MakeLinearColorFromVectorAndFloat(
     "Vector (RGB)"=&VectorFromCurve(VectorCurve=DataInterface() {propertyValues {XCurve {keys [<at 0 1>,<at 1 1>]}, YCurve {keys [<at 0 1>,<at 1 1>]}, ZCurve {keys [<at 0 1>,<at 1 1>]}}})
     "Float (Alpha)"=&ClampFloat(Float=&FloatFromCurve(FloatCurve=DataInterface() {propertyValues {Curve {keys [{arriveTangent=-1, time=0, value=1}, <at 1 0>]}}}))
    ))
   &DynamicMaterialParameters(
    Index 0 Param 1=&FloatFromCurve(FloatCurve=DataInterface() {propertyValues {Curve {keys [
         {arriveTangent=9.8869, time=0, value=0}
         {arriveTangent=19.9278, interpMode=RCIM_Cubic, leaveTangent=19.9278, time=0.3383, value=9.8869}
         {arriveTangent=12.2199, interpMode=RCIM_Cubic, leaveTangent=12.2199, time=0.5212, value=10.3864}
         {arriveTangent=15.5021, interpMode=RCIM_Cubic, leaveTangent=15.5021, time=0.6035, value=13.1282}
         {arriveTangent=148.7672, interpMode=RCIM_Cubic, leaveTangent=148.7672, time=0.8879, value=16.0713}
         <at 1 72.1094>
        ]}}})
    Index 0 Param 2=1
    Index 0 Param 3=1
    Index 0 Param 4=1
    Module.Index0Param0Write=true
    Module.Index0Param1Write=true
    Module.Index0Param2Write=true
    Module.Index0Param3Write=true
   )
  ]
  rendererClasses [&NiagaraMeshRendererProperties]
  renderers [{rendererClass=&NiagaraMeshRendererProperties, rendererIndex=0}]
 }
 MapLeft {
  emitterUpdateScript [
   &EmitterState(Life Cycle Mode=::NewEnumerator1, Loop Behavior=::NewEnumerator1, Loop Delay=0.401, Loop Duration=1, UseLoopDelay=true)
   &SpawnBurst_Instantaneous(Use Spawn Probability=false, UseLoopCountLimit=false)
  ]
  particleSpawnScript [
   /SetVariables(Particles.Acceleration:Vector3f=&RandomRangeVector(Minimum=<0 0 -450>, Maximum=<0 0 -980>)) {isSetParameters=true}
   /SetVariables(Particles.SpawnRotationVector=<0 0 -1>) {isSetParameters=true}
   &InitializeParticle(
    Lifetime=0.1
    Mesh Scale=<0.42 0.44 0.44>
    Mesh Scale Mode=::NewEnumerator4
    Position Offset=$System.SystemOffset
    UsePositionOffset=true
    Write Material Random=true
   )
   &AddVelocity(Velocity=&RandomRangeVector(Minimum=<-650 0 0>, Maximum=<-800 0 0>))
  ]
  particleUpdateScript [
   &ParticleState()
   &UpdateMeshOrientation(Rotation Vector=$Particles.SpawnRotationVector)
   &ScaleMeshSize(Scale Factor=&VectorFromCurve(VectorCurve=DataInterface() {propertyValues {
       XCurve {keys [
         {arriveTangent=-1.91, interpMode=RCIM_Cubic, time=0.1328, value=1.4207}
         {arriveTangent=-1.6381, interpMode=RCIM_Cubic, leaveTangent=-1.6381, time=0.5072, value=1}
         <at 1 0>
        ]}
       YCurve {keys [<at 0.1328 1 RCIM_Cubic>, {arriveTangent=-1.1531, interpMode=RCIM_Cubic, leaveTangent=-1.1531, time=0.5072, value=1}, <at 1 0>]}
       ZCurve {keys [<at 0.1328 1 RCIM_Cubic>, {arriveTangent=-1.1531, interpMode=RCIM_Cubic, leaveTangent=-1.1531, time=0.5072, value=1}, <at 1 0>]}
      }}))
   &Color(Color=&MakeLinearColorFromVectorAndFloat(
     "Vector (RGB)"=&VectorFromCurve(VectorCurve=DataInterface() {propertyValues {XCurve {keys [<at 0 1>,<at 1 1>]}, YCurve {keys [<at 0 1>,<at 1 1>]}, ZCurve {keys [<at 0 1>,<at 1 1>]}}})
     "Float (Alpha)"=&ClampFloat(Float=&FloatFromCurve(FloatCurve=DataInterface() {propertyValues {Curve {keys [{arriveTangent=-1, time=0, value=1}, <at 1 0>]}}}))
    ))
   &DynamicMaterialParameters(
    Index 0 Param 1=&FloatFromCurve(FloatCurve=DataInterface() {propertyValues {Curve {keys [<at 0 72>,<at 1 1>]}}})
    Index 0 Param 2=1
    Index 0 Param 3=1
    Index 0 Param 4=1
    Module.Index0Param0Write=true
    Module.Index0Param1Write=true
    Module.Index0Param2Write=true
    Module.Index0Param3Write=true
   )
   &AccelerationForce(Acceleration=$Particles.Acceleration)
   &Drag(Drag=4, Use Linear Drag=true, Use Rotational Drag=true)
   &SolveForcesAndVelocity(Clamp Velocity=false, Limit Acceleration=false)
  ]
  rendererClasses [&NiagaraMeshRendererProperties]
  renderers [{rendererClass=&NiagaraMeshRendererProperties, rendererIndex=0}]
 }
 MapLeft001 {
  emitterUpdateScript [
   &EmitterState(Life Cycle Mode=::NewEnumerator1, Loop Behavior=::NewEnumerator1, Loop Delay=0.401, Loop Duration=1, UseLoopDelay=true)
   &SpawnBurst_Instantaneous(Use Spawn Probability=false, UseLoopCountLimit=false)
  ]
  particleSpawnScript [
   /SetVariables(Particles.Acceleration:Vector3f=&RandomRangeVector(Minimum=<0 0 -450>, Maximum=<0 0 -980>)) {isSetParameters=true}
   /SetVariables(Particles.SpawnRotationVector=<0 0 1>) {isSetParameters=true}
   &InitializeParticle(
    Lifetime=0.1
    Mesh Scale=<0.42 0.44 0.44>
    Mesh Scale Mode=::NewEnumerator4
    Position Offset=$System.SystemOffset
    UsePositionOffset=true
    Write Material Random=true
   )
   &AddVelocity(Velocity=&RandomRangeVector(Minimum=<650 0 0>, Maximum=<800 0 0>))
  ]
  particleUpdateScript [
   &ParticleState()
   &UpdateMeshOrientation(Rotation Vector=$Particles.SpawnRotationVector)
   &ScaleMeshSize(Scale Factor=&VectorFromCurve(VectorCurve=DataInterface() {propertyValues {
       XCurve {keys [
         {arriveTangent=-1.91, interpMode=RCIM_Cubic, time=0.1328, value=1.4207}
         {arriveTangent=-1.6381, interpMode=RCIM_Cubic, leaveTangent=-1.6381, time=0.5072, value=1}
         <at 1 0>
        ]}
       YCurve {keys [<at 0.1328 1 RCIM_Cubic>, {arriveTangent=-1.1531, interpMode=RCIM_Cubic, leaveTangent=-1.1531, time=0.5072, value=1}, <at 1 0>]}
       ZCurve {keys [<at 0.1328 1 RCIM_Cubic>, {arriveTangent=-1.1531, interpMode=RCIM_Cubic, leaveTangent=-1.1531, time=0.5072, value=1}, <at 1 0>]}
      }}))
   &Color(Color=&MakeLinearColorFromVectorAndFloat(
     "Vector (RGB)"=&VectorFromCurve(VectorCurve=DataInterface() {propertyValues {XCurve {keys [<at 0 1>,<at 1 1>]}, YCurve {keys [<at 0 1>,<at 1 1>]}, ZCurve {keys [<at 0 1>,<at 1 1>]}}})
     "Float (Alpha)"=&ClampFloat(Float=&FloatFromCurve(FloatCurve=DataInterface() {propertyValues {Curve {keys [{arriveTangent=-1, time=0, value=1}, <at 1 0>]}}}))
    ))
   &DynamicMaterialParameters(
    Index 0 Param 1=&FloatFromCurve(FloatCurve=DataInterface() {propertyValues {Curve {keys [<at 0 72>,<at 1 1>]}}})
    Index 0 Param 2=1
    Index 0 Param 3=1
    Index 0 Param 4=1
    Module.Index0Param0Write=true
    Module.Index0Param1Write=true
    Module.Index0Param2Write=true
    Module.Index0Param3Write=true
   )
   &AccelerationForce(Acceleration=$Particles.Acceleration)
   &Drag(Drag=4, Use Linear Drag=true, Use Rotational Drag=true)
   &SolveForcesAndVelocity(Clamp Velocity=false, Limit Acceleration=false)
  ]
  rendererClasses [&NiagaraMeshRendererProperties]
  renderers [{rendererClass=&NiagaraMeshRendererProperties, rendererIndex=0}]
 }
 Smoke {
  emitterUpdateScript [
   &EmitterState(Life Cycle Mode=::NewEnumerator1, Loop Behavior=::NewEnumerator1, Loop Delay=0.4, Loop Duration=1, UseLoopDelay=true)
   &SpawnBurst_Instantaneous(Spawn Count=3, Use Spawn Probability=false, UseLoopCountLimit=false)
  ]
  particleSpawnScript [
   &InitializeParticle(
    Color=<rgba 0.8 0.5735 0.3375 0.5>
    Lifetime=&RandomRangeFloat(Minimum=0.25, Maximum=0.45)
    Position Offset=&Add_Vector(A=&RandomRangeVector(Minimum=<-6 -1.5 0>, Maximum=<6 1.5 0>), B=$System.SystemOffset)
    Sprite Rotation Mode=::NewEnumerator2
    Sprite Size=&Vector2DFromFloat(Value=&RandomRangeFloat(Minimum=35, Maximum=65))
    Sprite Size Mode=::NewEnumerator4
    UsePositionOffset=true
    Write Material Random=true
   )
   &CameraOffset(Camera Offset Amount=&RandomRangeFloat(Minimum=20, Maximum=12))
  ]
  particleUpdateScript [
   &ParticleState()
   &SpriteRotationRate(Rotation Rate=&AngleConversion(
     Angle:NiagaraFloat=&RandomRangeFloat(Minimum=2, Maximum=3)
     Angle Input:ENiagara_AngleInput=::NewEnumerator1
     Angle Output:ENiagara_AngleInput=::NewEnumerator2
    ))
   &ScaleSpriteSize(
    Scale Sprite Size Mode=::NewEnumerator0
    Uniform Scale Factor=&FloatFromCurve(FloatCurve=DataInterface() {propertyValues {Curve {keys [
         {arriveTangent=2.4508, interpMode=RCIM_Cubic, leaveTangent=2.4508, tangentMode=RCTM_User, time=0, value=0.1}
         {interpMode=RCIM_Cubic, leaveTangent=0.0935, time=1, value=1}
        ]}}})
   )
   &DynamicMaterialParameters(
    Index 0 Param 1=&FloatFromCurve(FloatCurve=DataInterface() {propertyValues {Curve {keys [<at 0.1859 1 RCIM_Cubic>,<at 1 0>]}}})
    Index 0 Param 2=1
    Index 0 Param 3=1
    Index 0 Param 4=1
    Module.Index0Param0Write=true
    Module.Index0Param1Write=true
    Module.Index0Param2Write=true
    Module.Index0Param3Write=true
   )
   &ScaleColor(
    Scale Alpha=&FloatFromCurve(FloatCurve=DataInterface() {propertyValues {Curve {keys [
         {arriveTangent=0.9919, interpMode=RCIM_Cubic, tangentMode=RCTM_User, time=0, value=0}
         {interpMode=RCIM_Cubic, tangentMode=RCTM_User, time=0.0436, value=0.9919}
         {interpMode=RCIM_Cubic, tangentMode=RCTM_User, time=1, value=0}
        ]}}})
    ScaleA=true
    ScaleRGB=true
   )
   &Drag(Drag=3, Use Linear Drag=true, Use Rotational Drag=true)
   &SolveForcesAndVelocity(Clamp Velocity=false, Limit Acceleration=false)
   /Niagara/Modules/Update/SubUV/V2/SubUVAnimation(
    Random Start Frame=true
    "Sprite Renderer "=DataInterface() {propertyValues {SpriteRenderer=&self:Smoke_6.NiagaraSpriteRendererProperties_3}}
    SubUV Lookup Index Scale=0.1
    UseEndFrame=true
    UseStartFrame=true
   )
  ]
  rendererClasses [&NiagaraSpriteRendererProperties]
  renderers [{rendererClass=&NiagaraSpriteRendererProperties}]
 }
 SmokeSwirl {
  emitterUpdateScript [
   &EmitterState(Life Cycle Mode=::NewEnumerator1, Loop Behavior=::NewEnumerator1, Loop Delay=0.4, Loop Duration=1, UseLoopDelay=true)
   &SpawnBurst_Instantaneous(Spawn Count=2, Use Spawn Probability=false, UseLoopCountLimit=false)
  ]
  particleSpawnScript [&InitializeParticle(
    Color=<rgba 1 0.36 0.06 1>
    Lifetime=&RandomRangeFloat(Minimum=0.4, Maximum=0.65)
    Position Offset=&Add_Vector(A=&RandomRangeVector(Minimum=<-6 -1.5 0>, Maximum=<6 1.5 0>), B=$System.SystemOffset)
    Sprite Rotation Angle=&Lerp_Float(A=0.25, B=0.75, Alpha=&ReturnNormalizedExecIndex())
    Sprite Rotation Mode=::NewEnumerator1
    Sprite Size Mode=::NewEnumerator3
    Uniform Sprite Size=&RandomRangeFloat(Minimum=40, Maximum=55)
    UsePositionOffset=true
    Write Material Random=true
   )]
  particleUpdateScript [
   &ParticleState()
   &SpriteRotationRate(Rotation Rate=&AngleConversion(
     Angle:NiagaraFloat=&RandomRangeFloat(Minimum=-0.03, Maximum=0.003)
     Angle Input:ENiagara_AngleInput=::NewEnumerator1
     Angle Output:ENiagara_AngleInput=::NewEnumerator2
    ))
   &ScaleSpriteSize(
    Scale Sprite Size Mode=::NewEnumerator0
    Uniform Scale Factor=&FloatFromCurve(FloatCurve=DataInterface() {propertyValues {Curve {keys [
         {arriveTangent=2.503, interpMode=RCIM_Cubic, leaveTangent=2.503, tangentMode=RCTM_User, time=0, value=0.5}
         {interpMode=RCIM_Cubic, tangentMode=RCTM_User, time=1, value=2}
        ]}}})
   )
   &ScaleColor(
    Scale Alpha=&FloatFromCurve(FloatCurve=DataInterface() {propertyValues {Curve {keys [
         {arriveTangent=4.9805, interpMode=RCIM_Cubic, time=0, value=0}
         {arriveTangent=1.0493, interpMode=RCIM_Cubic, leaveTangent=1.0493, time=0.1487, value=0.9942}
         {arriveTangent=-1.1678, interpMode=RCIM_Cubic, leaveTangent=-1.1678, time=0.2827, value=0.2966}
         {interpMode=RCIM_Cubic, leaveTangent=-0.0162, time=1, value=0}
        ]}}})
    Scale RGB=<3 3 3>
    ScaleA=true
    ScaleRGB=true
   )
   &Drag(Drag=2, Use Linear Drag=true, Use Rotational Drag=true)
   &SolveForcesAndVelocity(Clamp Velocity=false, Limit Acceleration=false)
   &SubUVAnimation(
    "Sprite Renderer "=DataInterface() {propertyValues {SpriteRenderer=&self:SmokeSwirl_5.NiagaraSpriteRendererProperties_2}}
    SubUV Lookup Index Scale=0.5
    UseEndFrame=true
    UseStartFrame=true
   )
   &DynamicMaterialParameters(
    Index 0 Param 1=&FloatFromCurve(FloatCurve=DataInterface() {propertyValues {Curve {keys [<at 0 0>,<at 1 1>]}}})
    Index 0 Param 2=1
    Index 0 Param 3=1
    Index 0 Param 4=1
    Module.Index0Param0Write=true
    Module.Index0Param1Write=true
    Module.Index0Param2Write=true
    Module.Index0Param3Write=true
   )
  ]
  rendererClasses [&NiagaraSpriteRendererProperties]
  renderers [{rendererClass=&NiagaraSpriteRendererProperties}]
 }
}
summary {
 systemName=NS_PirateMap_TearUp
 emitters [
  {emitterName=Map, rendererClasses [&NiagaraMeshRendererProperties]}
  {emitterName=MapLeft, rendererClasses [&NiagaraMeshRendererProperties]}
  {emitterName=MapLeft001, rendererClasses [&NiagaraMeshRendererProperties]}
  {emitterName=Flash, rendererClasses [&NiagaraMeshRendererProperties]}
  {emitterName=Ash, rendererClasses [&NiagaraSpriteRendererProperties]}
  {emitterName=Confetti_Center, simTarget=GPUComputeSim, rendererClasses [&NiagaraSpriteRendererProperties]}
  {emitterName=SmokeSwirl, rendererClasses [&NiagaraSpriteRendererProperties]}
  {emitterName=Smoke, rendererClasses [&NiagaraSpriteRendererProperties]}
 ]
 userVariables []
}
